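# Sound cue for spawning a building prefab with a Creative gadget, routed to the player construction sound class.
# It layers two prefab spawn waves with one of three randomly chosen wood construction-finish waves.
# The mix gets random pitch variation, then passes through an overriding attenuation node with distance low-pass filtering and stereo spread.
# The cue plays at half volume and shares the player wall-build concurrency rule.
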SoundCue /Game/Sounds/Creative/Gadgets/Building/Building_Prefab_Spawn_Cue
root=SoundClass_1
nodes [
 {
  id=Attenuation_0
  children [Modulator_9]
  properties {
   bOverrideAttenuation=true
   attenuationOverrides {
    bAttenuateWithLPF=true
    falloffDistance=10000
    lPFFrequencyAtMin=800
    lPFRadiusMax=8000
    lPFRadiusMin=4000
    stereoSpread=2000
    attenuationShapeExtents {x=2000}
   }
  }
 }
 {id=Mixer_6, children [Random_0, WavePlayer_16, WavePlayer_17], properties {inputVolume [1,1,1]}}
 {id=Modulator_9, children [Mixer_6], properties {pitchMax=1.3, pitchMin=0.9, volumeMax=1, volumeMin=1}}
 {id=Random_0, children [WavePlayer_0,WavePlayer_1,WavePlayer_2], properties {weights [1,1,1]}}
 {id=SoundClass_1, children [Attenuation_0], properties {soundClassOverride=/Game/Sounds/FortSoundClassesAndModes/Fort_Player_Construction}}
 {id=WavePlayer_0, properties {soundWaveAssetPtr=/Game/Sounds/Fort_UI_Sounds/Gameplay/BuildingSounds/Wavs/Fort_Construction_Done_Wood_01}}
 {id=WavePlayer_1, properties {soundWaveAssetPtr=/Game/Sounds/Fort_UI_Sounds/Gameplay/BuildingSounds/Wavs/Fort_Construction_Done_Wood_02}}
 {id=WavePlayer_16, properties {soundWaveAssetPtr=&/Building_Prefab_Spawn01}}
 {id=WavePlayer_17, properties {soundWaveAssetPtr=&/Building_Prefab_Spawn02}}
 {id=WavePlayer_2, properties {soundWaveAssetPtr=/Game/Sounds/Fort_UI_Sounds/Gameplay/BuildingSounds/Wavs/Fort_Construction_Done_Wood_03}}
]
properties {
 maxDistance=12000
 volumeMultiplier=0.5
 attenuationOverrides {falloffDistance=2600}
 concurrencyOverrides {resolutionRule=PreventNew}
 concurrencySet [/Game/Sounds/Concurrency_Rules/Player_Wall_Build_Conc]
}
